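# Animation Blueprint for the female medium DarkVogue cosmetic body, adding secondary motion to clothing and accessories.
# It copies the attached parent's pose, runs a Control Rig fed Z velocity and the inverted rigid-body flag, then weight-constrains the zipper bone.
# Chained AnimDynamics spring jacket (with spherical limits), zipper, dagger, laces, lapels and linear-sprung hood; two duplicate nodes are unconnected.
# EventGraph updates those inputs and a deprecated-function gravity override, unread here.
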AnimBlueprint /BRCosmetics/Characters/Player/Female/Medium/Bodies/F_MED_DarkVogue/Meshes/F_MED_DarkVogue_AnimBP
target_skeleton=&/F_MED_DarkVogue_Skeleton
graphs [
 AnimGraph {nodes [
   Anim Dynamics - Bone: Dyn_hood_Back(Alpha=1, ExternalForce="0,0,0") {
    angularDampingOverride=0.9999
    angularSpringConstant=50
    bAngularSpring=true
    bGravityOverrideInSimSpace=true
    bOverrideAngularDamping=true
    bUseGravityOverride=true
    simulationSpace=World
    alphaBoolBlend {blendOption=Linear}
    boundBone {boneName=Dyn_hood_Back}
    componentAppliedLinearAccClamp=<10000 10000 10000>
    componentLinearAccScale=<0.7 0.7 0.7>
    componentLinearVelScale=<0.7 0.7 0.7>
    gravityOverride {z=-980}
    physicsBodyDefinitions [{
      sphereCollisionRadius=5
      boundBone {boneName=Dyn_hood_Back}
      boxExtents=<6 6 6>
      constraintSetup {angularLimitsMax=<10 15 0>, angularLimitsMin=<-10 -15 -30>, angularTarget=<1 0 0>}
      localJointOffset=<0.8 0 0>
     }]
    relativeSpaceBone {boneName=spine_05}
   }
   Anim Dynamics - Bone: Dyn_dagger_R(Alpha=1, ExternalForce="0,0,0") {
    angularDampingOverride=1
    angularSpringConstant=1000
    bAngularSpring=true
    bOverrideAngularDamping=true
    bUseGravityOverride=true
    simulationSpace=BoneRelative
    alphaBoolBlend {blendOption=Linear}
    boundBone {boneName=Dyn_dagger_R}
    componentAppliedLinearAccClamp=<10000 10000 10000>
    componentLinearAccScale=<0.7 0.7 0.5>
    componentLinearVelScale=<0.7 0.7 0.5>
    gravityOverride {z=-980}
    physicsBodyDefinitions [{
      boundBone {boneName=Dyn_dagger_R}
      boxExtents=<6 20 3>
      constraintSetup {angularLimitsMax=<6 4 7>, angularLimitsMin=<-6 -5 -7>, angularTarget=<1 0 0>}
      localJointOffset=<-0.6 0 0.3>
     }]
    relativeSpaceBone {boneName=calf_r}
   }
   n1 = Control Rig!(
    Source=Copy Pose From Mesh() {bUseAttachedParent=true}
    Alpha
    Velocity_Z:double=Get Velocity Z() -> (Velocity Z:double)
    NOT_RigidBodyEnabled:bool=Get NOT_RigidBodyEnabled() -> (NOT_RigidBodyEnabled:bool)
   ) -> (Pose) {
    alphaBoolBlend {blendOption=Linear}
    controlRigAssetReference {blueprintRigClass=&/F_MED_DarkVogue_CtrlRig.F_MED_DarkVogue_CtrlRig_C}
   }
   n2 = "Constraint - Dyn_zipper (PPPPPPPP)"!(
    ComponentPose=Local To Component(LocalPose=@n1)
    Alpha=1
    ConstraintWeights_0=0.2451
    ConstraintWeights_1:float=0.0364
    ConstraintWeights_2:float=0.0581
    ConstraintWeights_3:float=0.0077
    ConstraintWeights_4:float=0.0058
    ConstraintWeights_5:float=0.0038
    ConstraintWeights_6:float=0.0167
    ConstraintWeights_7:float=0.6264
   ) -> (Pose) {
    alphaBoolBlend {blendOption=Linear}
    boneToModify {boneName=Dyn_zipper}
    constraintSetup [
     {targetBone {boneName=spine_04}}
     {targetBone {boneName=spine_03}}
     {targetBone {boneName=spine_02}}
     {targetBone {boneName=spine_01}}
     {targetBone {boneName=pelvis}}
     {targetBone {boneName=deform_pec_l}}
     {targetBone {boneName=Dyn_jacket_L}}
     {targetBone {boneName=Dyn_jacket_Front_L}}
    ]
    constraintWeights [1,1,1,1,1,1,1,1]
   }
   n3 = Anim Dynamics - Bone: Dyn_jacket_R(ComponentPose=@n2, Alpha=1, ExternalForce="0,0,0") {
    angularDampingOverride=1
    angularSpringConstant=500
    bAngularSpring=true
    bGravityOverrideInSimSpace=true
    bOverrideAngularDamping=true
    bUseGravityOverride=true
    bUseSphericalLimits=true
    simulationSpace=RootRelative
    alphaBoolBlend {blendOption=Linear}
    boundBone {boneName=Dyn_jacket_R}
    componentAppliedLinearAccClamp=<5000 5000 5000>
    componentLinearAccScale=<0.5 0.5 0.5>
    componentLinearVelScale=<0.5 0.5 0.5>
    gravityOverride {z=-980}
    physicsBodyDefinitions [{
      collisionType=CustomSphere
      sphereCollisionRadius=3
      boundBone {boneName=Dyn_jacket_R}
      boxExtents=<6 6 6>
      constraintSetup {angularLimitsMax=<0 0 5>, angularLimitsMin=<0 -15 -5>, angularTarget=<1 0 0>}
      localJointOffset=<-0.2 0 -0.3>
     }]
    relativeSpaceBone {boneName=spine_04}
    sphericalLimits [{limitRadius=9, drivingBone {boneName=spine_03}, sphereLocalOffset=<0 0 0>}]
   }
   n4 = Anim Dynamics - Bone: Dyn_jacket_L(ComponentPose=@n3, Alpha=1, ExternalForce="0,0,0") {
    angularDampingOverride=1
    angularSpringConstant=500
    bAngularSpring=true
    bGravityOverrideInSimSpace=true
    bOverrideAngularDamping=true
    bUseGravityOverride=true
    bUseSphericalLimits=true
    simulationSpace=RootRelative
    alphaBoolBlend {blendOption=Linear}
    boundBone {boneName=Dyn_jacket_L}
    componentAppliedLinearAccClamp=<5000 5000 5000>
    componentLinearAccScale=<0.5 0.5 0.5>
    componentLinearVelScale=<0.5 0.5 0.5>
    gravityOverride {z=-980}
    physicsBodyDefinitions [{
      collisionType=CustomSphere
      sphereCollisionRadius=3
      boundBone {boneName=Dyn_jacket_L}
      boxExtents=<6 6 6>
      constraintSetup {angularLimitsMax=<0 -3 5>, angularLimitsMin=<0 -18 -5>, angularTarget=<1 0 0>}
      localJointOffset=<-0.2 0 0.3>
     }]
    relativeSpaceBone {boneName=spine_04}
    sphericalLimits [{limitRadius=9, drivingBone {boneName=spine_03}, sphereLocalOffset=<0 0 0>}]
   }
   n5 = Anim Dynamics - Bone: Dyn_jacket_Back(ComponentPose=@n4, Alpha=1, ExternalForce="0,0,0") {
    angularDampingOverride=1
    angularSpringConstant=500
    bAngularSpring=true
    bGravityOverrideInSimSpace=true
    bOverrideAngularDamping=true
    bUseGravityOverride=true
    bUsePlanarLimit=false
    bUseSphericalLimits=true
    simulationSpace=RootRelative
    alphaBoolBlend {blendOption=Linear}
    boundBone {boneName=Dyn_jacket_Back}
    componentAppliedLinearAccClamp=<5000 5000 5000>
    componentLinearAccScale=<0.5 0.5 0.5>
    componentLinearVelScale=<0.5 0.5 0.5>
    gravityOverride {z=-980}
    physicsBodyDefinitions [{
      collisionType=CustomSphere
      sphereCollisionRadius=2.5
      boundBone {boneName=Dyn_jacket_Back}
      boxExtents=<6 6 6>
      constraintSetup {angularLimitsMax=<0 0 0>, angularLimitsMin=<-10 -18 0>, angularTarget=<1 0 0>}
      localJointOffset=<0 0 0.6>
     }]
    relativeSpaceBone {boneName=spine_04}
    sphericalLimits [{limitRadius=7.5, drivingBone {boneName=spine_03}, sphereLocalOffset=<0 2.1059 0>}]
   }
   n6 = Anim Dynamics - Bone: Dyn_jacket_Front_R(ComponentPose=@n5, Alpha=1, ExternalForce="0,0,0") {
    angularDampingOverride=1
    angularSpringConstant=500
    bAngularSpring=true
    bGravityOverrideInSimSpace=true
    bOverrideAngularDamping=true
    bUseGravityOverride=true
    bUsePlanarLimit=false
    bUseSphericalLimits=true
    simulationSpace=RootRelative
    alphaBoolBlend {blendOption=Linear}
    boundBone {boneName=Dyn_jacket_Front_R}
    componentAppliedLinearAccClamp=<5000 5000 5000>
    componentLinearAccScale=<0.55 0.55 0.55>
    componentLinearVelScale=<0.5 0.5 0.5>
    gravityOverride {z=-980}
    physicsBodyDefinitions [{
      collisionType=CustomSphere
      sphereCollisionRadius=3
      boundBone {boneName=Dyn_jacket_Front_R}
      boxExtents=<6 6 6>
      constraintSetup {angularLimitsMax=<0 0 4>, angularLimitsMin=<-5 -7 -4>, angularTarget=<1 0 0>}
      localJointOffset=<0.4 0 0>
     }]
    relativeSpaceBone {boneName=spine_04}
    sphericalLimits [{limitRadius=9, drivingBone {boneName=spine_03}, sphereLocalOffset=<0 0 0>}]
   }
   n7 = Anim Dynamics - Bone: Dyn_jacket_Front_L(ComponentPose=@n6, Alpha=1, ExternalForce="0,0,0") {
    angularDampingOverride=1
    angularSpringConstant=500
    bAngularSpring=true
    bGravityOverrideInSimSpace=true
    bOverrideAngularDamping=true
    bUseGravityOverride=true
    bUseSphericalLimits=true
    simulationSpace=RootRelative
    alphaBoolBlend {blendOption=Linear}
    boundBone {boneName=Dyn_jacket_Front_L}
    componentAppliedLinearAccClamp=<5000 5000 5000>
    componentLinearAccScale=<0.55 0.55 0.55>
    componentLinearVelScale=<0.5 0.5 0.5>
    gravityOverride {z=-980}
    physicsBodyDefinitions [{
      collisionType=CustomSphere
      sphereCollisionRadius=3
      boundBone {boneName=Dyn_jacket_Front_L}
      boxExtents=<6 6 6>
      constraintSetup {angularLimitsMax=<0 0 4>, angularLimitsMin=<-5 -7 -4>, angularTarget=<1 0 0>}
      localJointOffset=<-0.4 0 0>
     }]
    relativeSpaceBone {boneName=spine_04}
    sphericalLimits [{limitRadius=9, drivingBone {boneName=spine_03}, sphereLocalOffset=<0 0 0>}]
   }
   n8 = Anim Dynamics - Bone: Dyn_zipper(ComponentPose=@n7, Alpha=1, ExternalForce="0,0,0") {
    angularDampingOverride=0.9999
    angularSpringConstant=50
    bAngularSpring=true
    bGravityOverrideInSimSpace=true
    bOverrideAngularDamping=true
    bUseGravityOverride=true
    simulationSpace=RootRelative
    alphaBoolBlend {blendOption=Linear}
    boundBone {boneName=Dyn_zipper}
    componentAppliedLinearAccClamp=<10000 10000 10000>
    componentLinearAccScale=<0.7 0.7 0.7>
    componentLinearVelScale=<0.7 0.7 0.7>
    gravityOverride {z=-980}
    physicsBodyDefinitions [{
      boundBone {boneName=Dyn_zipper}
      boxExtents=<5 5 5>
      constraintSetup {angularTargetAxis=AxisZ, angularLimitsMax=<15 5 25>, angularLimitsMin=<-15 -60 -25>, angularTarget=<0 0 1>}
      localJointOffset=<0.5 0 0>
     }]
    relativeSpaceBone {boneName=root}
   }
   n9 = Anim Dynamics - Bone: Dyn_dagger_R(ComponentPose=@n8, Alpha=1, ExternalForce="0,0,0") {
    angularDampingOverride=1
    angularSpringConstant=1000
    bAngularSpring=true
    bOverrideAngularDamping=true
    bUseGravityOverride=true
    simulationSpace=RootRelative
    alphaBoolBlend {blendOption=Linear}
    boundBone {boneName=Dyn_dagger_R}
    componentAppliedLinearAccClamp=<10000 10000 10000>
    componentLinearAccScale=<0.5 0.5 0.2>
    componentLinearVelScale=<0.5 0.5 0.2>
    gravityOverride {z=-980}
    physicsBodyDefinitions [{
      boundBone {boneName=Dyn_dagger_R}
      boxExtents=<7 7 7>
      constraintSetup {angularTargetAxis=AxisZ, angularLimitsMax=<6 4 7>, angularLimitsMin=<-6 -5 -7>, angularTarget=<0 0 1>}
      localJointOffset=<0.4 -0.05 0>
     }]
    relativeSpaceBone {boneName=calf_r}
   }
   n10 = Anim Dynamics - Bone: laces_l_L(ComponentPose=@n9, Alpha=1, ExternalForce="0,0,0") {
    angularDampingOverride=0.9999
    angularSpringConstant=500
    bAngularSpring=true
    bGravityOverrideInSimSpace=true
    bOverrideAngularDamping=true
    bUseGravityOverride=true
    simulationSpace=World
    alphaBoolBlend {blendOption=Linear}
    boundBone {boneName=laces_l_L}
    componentAppliedLinearAccClamp=<10000 10000 10000>
    componentLinearAccScale=<0.5 0.5 0.5>
    componentLinearVelScale=<0.5 0.5 0.5>
    gravityOverride {z=-980}
    physicsBodyDefinitions [{
      boundBone {boneName=laces_l_L}
      boxExtents=<6 6 6>
      constraintSetup {angularLimitsMax=<15 0 20>, angularLimitsMin=<-15 -20 -20>, angularTarget=<1 0 0>}
      localJointOffset=<0.5 0.5 0>
     }]
    relativeSpaceBone {boneName=foot_l}
   }
   n11 = Anim Dynamics - Bone: laces_r_L(ComponentPose=@n10, Alpha=1, ExternalForce="0,0,0") {
    angularDampingOverride=0.9999
    angularSpringConstant=500
    bAngularSpring=true
    bGravityOverrideInSimSpace=true
    bOverrideAngularDamping=true
    bUseGravityOverride=true
    simulationSpace=World
    alphaBoolBlend {blendOption=Linear}
    boundBone {boneName=laces_r_L}
    componentAppliedLinearAccClamp=<10000 10000 10000>
    componentLinearAccScale=<0.5 0.5 0.5>
    componentLinearVelScale=<0.5 0.5 0.5>
    gravityOverride {z=-980}
    physicsBodyDefinitions [{
      boundBone {boneName=laces_r_L}
      boxExtents=<6 6 6>
      constraintSetup {angularLimitsMax=<15 0 20>, angularLimitsMin=<-15 -20 -20>, angularTarget=<1 0 0>}
      localJointOffset=<0.5 0.5 0>
     }]
    relativeSpaceBone {boneName=foot_l}
   }
   n12 = Anim Dynamics - Bone: laces_c_L(ComponentPose=@n11, Alpha=1, ExternalForce="0,0,0") {
    angularDampingOverride=0.9999
    angularSpringConstant=500
    bAngularSpring=true
    bGravityOverrideInSimSpace=true
    bOverrideAngularDamping=true
    bUseGravityOverride=true
    simulationSpace=World
    alphaBoolBlend {blendOption=Linear}
    boundBone {boneName=laces_c_L}
    componentAppliedLinearAccClamp=<10000 10000 10000>
    componentLinearAccScale=<0.3 0.3 0.3>
    componentLinearVelScale=<0.3 0.3 0.3>
    gravityOverride {z=-980}
    physicsBodyDefinitions [{
      boundBone {boneName=laces_c_L}
      boxExtents=<6 6 6>
      constraintSetup {angularTargetAxis=AxisZ, angularLimitsMax=<15 -3 20>, angularLimitsMin=<-15 -25 -20>, angularTarget=<0 0 1>}
      localJointOffset=<0.3 0 0.3>
     }]
    relativeSpaceBone {boneName=foot_r}
   }
   n13 = Anim Dynamics - Bone: laces_l_R(ComponentPose=@n12, Alpha=1, ExternalForce="0,0,0") {
    angularDampingOverride=0.9999
    angularSpringConstant=500
    bAngularSpring=true
    bGravityOverrideInSimSpace=true
    bOverrideAngularDamping=true
    bUseGravityOverride=true
    simulationSpace=World
    alphaBoolBlend {blendOption=Linear}
    boundBone {boneName=laces_l_R}
    componentAppliedLinearAccClamp=<10000 10000 10000>
    componentLinearAccScale=<0.5 0.5 0.5>
    componentLinearVelScale=<0.5 0.5 0.5>
    gravityOverride {z=-980}
    physicsBodyDefinitions [{
      boundBone {boneName=laces_l_R}
      boxExtents=<6 6 6>
      constraintSetup {angularLimitsMax=<15 0 20>, angularLimitsMin=<-15 -20 -20>, angularTarget=<1 0 0>}
      localJointOffset=<-0.5 -0.5 0>
     }]
    relativeSpaceBone {boneName=foot_r}
   }
   n14 = Anim Dynamics - Bone: laces_r_R(ComponentPose=@n13, Alpha=1, ExternalForce="0,0,0") {
    angularDampingOverride=0.9999
    angularSpringConstant=500
    bAngularSpring=true
    bGravityOverrideInSimSpace=true
    bOverrideAngularDamping=true
    bUseGravityOverride=true
    simulationSpace=World
    alphaBoolBlend {blendOption=Linear}
    boundBone {boneName=laces_r_R}
    componentAppliedLinearAccClamp=<10000 10000 10000>
    componentLinearAccScale=<0.5 0.5 0.5>
    componentLinearVelScale=<0.5 0.5 0.5>
    gravityOverride {z=-980}
    physicsBodyDefinitions [{
      boundBone {boneName=laces_r_R}
      boxExtents=<6 6 6>
      constraintSetup {angularLimitsMax=<15 0 20>, angularLimitsMin=<-15 -20 -20>, angularTarget=<1 0 0>}
      localJointOffset=<-0.5 -0.5 0>
     }]
    relativeSpaceBone {boneName=foot_r}
   }
   n15 = Anim Dynamics - Bone: laces_c_R(ComponentPose=@n14, Alpha=1, ExternalForce="0,0,0") {
    angularDampingOverride=0.9999
    angularSpringConstant=500
    bAngularSpring=true
    bGravityOverrideInSimSpace=true
    bOverrideAngularDamping=true
    bUseGravityOverride=true
    simulationSpace=World
    alphaBoolBlend {blendOption=Linear}
    boundBone {boneName=laces_c_R}
    componentAppliedLinearAccClamp=<10000 10000 10000>
    componentLinearAccScale=<0.3 0.3 0.3>
    componentLinearVelScale=<0.3 0.3 0.3>
    gravityOverride {z=-980}
    physicsBodyDefinitions [{
      boundBone {boneName=laces_c_R}
      boxExtents=<6 6 6>
      constraintSetup {angularTargetAxis=AxisZ, angularLimitsMax=<15 -5 20>, angularLimitsMin=<-15 -25 -20>, angularTarget=<0 0 1>}
      localJointOffset=<-0.3 0 -0.3>
     }]
    relativeSpaceBone {boneName=foot_r}
   }
   n16 = Anim Dynamics - Bone: Dyn_lapels(ComponentPose=@n15, Alpha=1, ExternalForce="0,0,0") {
    angularDampingOverride=0.9999
    angularSpringConstant=400
    bAngularSpring=true
    bGravityOverrideInSimSpace=true
    bOverrideAngularDamping=true
    bUseGravityOverride=true
    simulationSpace=World
    alphaBoolBlend {blendOption=Linear}
    boundBone {boneName=Dyn_lapels}
    componentAppliedLinearAccClamp=<10000 10000 10000>
    componentLinearAccScale {x=0.7, z=0.7}
    componentLinearVelScale {x=0.7, z=0.7}
    gravityOverride {z=-980}
    physicsBodyDefinitions [{
      sphereCollisionRadius=5
      boundBone {boneName=Dyn_lapels}
      boxExtents=<6 6 6>
      constraintSetup {angularLimitsMax=<10 8 0>, angularLimitsMin=<-10 -8 -30>, angularTarget=<1 0 0>}
      localJointOffset=<0 0.5 0>
     }]
    relativeSpaceBone {boneName=spine_05}
   }
   n17 = Anim Dynamics - Bone: Dyn_hood_Back(ComponentPose=@n16, Alpha=1, ExternalForce="0,0,0") {
    angularDampingOverride=0.9999
    angularSpringConstant=50
    bGravityOverrideInSimSpace=true
    bLinearSpring=true
    bOverrideLinearDamping=true
    bUseGravityOverride=true
    linearDampingOverride=0.9999
    linearSpringConstant=250
    simulationSpace=World
    alphaBoolBlend {blendOption=Linear}
    boundBone {boneName=Dyn_hood_Back}
    componentAppliedLinearAccClamp=<10000 10000 5000>
    gravityOverride {z=-980}
    physicsBodyDefinitions [{
      sphereCollisionRadius=5
      boundBone {boneName=Dyn_hood_Back}
      boxExtents=<12 12 12>
      constraintSetup {
       angularTargetAxis=AxisZ
       angularLimitsMax=<0 0 0>
       angularLimitsMin=<0 0 0>
       angularTarget=<0 0 1>
       linearAxesMax=<1 2 4>
       linearAxesMin=<-1 -2 -1>
      }
      localJointOffset=<0 0 0>
     }]
    relativeSpaceBone {boneName=spine_05}
   }
   Output Pose(Result=Component To Local(ComponentPose=@n17))
  ]}
 EventGraph {nodes [
   Event BlueprintUpdateAnimation()
   > n1 = Gravity Override State Function DEPRECATED(
    CharacterPartAnimInstance=Self-Reference()
    GameTimeInSeconds=GetGameTimeInSeconds()
    GravityOverrideParams=Get Gravity Override Params()
    DebugJointName=dyn_hackBottomArmor
   )
   > Set Override World Gravity(Override World Gravity:Vector=@n1) -> (Output_Get:Vector)
   > Set Velocity Z(Velocity Z:double=Get PawnVelocityZ()) -> (Output_Get:double)
   > Set NOT_RigidBodyEnabled(NOT_RigidBodyEnabled:bool=NOT Boolean(A=Get bRigidBodyEnabled())) -> (Output_Get:bool)
   TryGetPawnOwner()
  ]}
]
variables [
 Gravity Override Params {type=GravityOverrideParamsStruct}
 NOT_RigidBodyEnabled {type=boolean}
 Override World Gravity {type=Vector}
 Velocity Z {type=number}
]
