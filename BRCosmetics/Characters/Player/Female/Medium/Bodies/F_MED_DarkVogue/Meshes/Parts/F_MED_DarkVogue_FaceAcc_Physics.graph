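# Physics asset for the F_MED_DarkVogue face accessory part, defining capsule bodies and constraints for a braid and a hair chain.
# Dyn_braid bones hang from a base locked to spine_05, and Dyn_hair bones chain from Base_hair_01 under the head, with swing and twist limits tightening toward the tips.
# Torso, neck, head, hood, arm and thigh capsules supply body volumes; their constraints are free, and the thighs have none.
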PhysicsAsset /BRCosmetics/Characters/Player/Female/Medium/Bodies/F_MED_DarkVogue/Meshes/Parts/F_MED_DarkVogue_FaceAcc_Physics
bodies {
 Base_hair_01 [{
   shapeType=Capsule
   length=1.2883
   radius=0.8934
   shapeName=Base_hair_01_capsule
   center=<1.4234 0.033 0.2366>
   rotation=<rot 0 -95.0673 -64.0039>
  }]
 Correct_hood_Back [{shapeType=Capsule, length=26.7349, radius=10.128, shapeName=Correct_hood_Back_capsule, center=<-2.4977 0 0>, rotation=<rot 0 0 90>}]
 Dyn_braid_01 [{shapeType=Capsule, length=1.9792, radius=3.227, center=<3.6204 -0.3 0.5427>, rotation=<rot 90 0 0>}]
 Dyn_braid_02 [{shapeType=Capsule, length=4, radius=3, shapeName=Dyn_braid_02_capsule, center=<2.9646 0 0>, rotation=<rot 90 0 0>}]
 Dyn_braid_03 [{
   shapeType=Capsule
   length=4.2435
   radius=2.7565
   shapeName=Dyn_braid_03_capsule
   center=<3.1767 0 -0.4397>
   rotation=<rot 75.0002 -0.0001 -0.0002>
  }]
 Dyn_braid_04 [{shapeType=Capsule, length=5.4039, radius=3, shapeName=Dyn_braid_04_capsule, center=<5.7737 0 0>, rotation=<rot 90 0 0>}]
 Dyn_braid_Base [{shapeType=Capsule, length=2.1166, radius=2.1109, shapeName=Dyn_braid_Base_capsule, rotation=<rot 0 -90 90>}]
 Dyn_hair_01 [{shapeType=Capsule, length=2.3464, radius=1.1891, shapeName=Dyn_hair_01_capsule, center=<1.3474 0 0>, rotation=<rot 90 0 0>}]
 Dyn_hair_02 [{shapeType=Capsule, length=1.4347, radius=1.1456, shapeName=Dyn_hair_02_capsule, center=<1.4866 0 0>, rotation=<rot 90 0 0>}]
 Dyn_hair_03 [{shapeType=Capsule, length=2.3325, radius=1.035, shapeName=Dyn_hair_03_capsule, center=<1.7966 0 0>, rotation=<rot 90 0 0>}]
 Master_Correct_Hair [{
   shapeType=Capsule
   length=0.505
   radius=0.505
   shapeName=Master_Correct_Hair_capsule
   center=<2.2999 0.6072 1.486>
   rotation=<rot 0 58.5281 -89.8658>
  }]
 clavicle_l [{
   shapeType=Capsule
   length=4.2487
   radius=5.2834
   shapeName=clavicle_l_capsule
   center=<3.8601 -0.407 0.2582>
   rotation=<rot -85 179.9974 175.8258>
  }]
 head [{shapeType=Capsule, length=2.6628, radius=7.1479, rotation=<rot 90 0 0>}]
 lowerarm_l [{shapeType=Capsule, length=14.1982, radius=4.6124, shapeName=lowerarm_l_capsule, center=<9.4079 0 0>, rotation=<rot -90 180 180>}]
 lowerarm_r [{shapeType=Capsule, length=14.1982, radius=4.6124, shapeName=lowerarm_l_capsule, center=<-9.4079 0 0>, rotation=<rot -90 0 180>}]
 neck_01 [{shapeType=Capsule, length=18.3908, radius=4.6585, shapeName=neck_01_capsule, rotation=<rot 0 -20.0001 0>}]
 neck_02 [{shapeType=Capsule, length=2.6986, radius=3.7787, shapeName=neck_02_capsule}]
 spine_02 [{shapeType=Capsule, length=7.5351, radius=9.3848, shapeName=spine_02_capsule, center=<0.267 0 0>}]
 spine_03 [{shapeType=Capsule, length=10.1969, radius=8.5991, shapeName=spine_03_capsule, center=<0.6106 0.2666 0>}]
 spine_04 [{shapeType=Capsule, length=11.8706, radius=9.3858, shapeName=spine_04_capsule, center=<2.3907 0.198 0>}]
 spine_05 [
  {shapeType=Capsule, length=8.5745, radius=6.9919, shapeName=spine_05_capsule, center=<0 0.5022 0>}
  {shapeType=Capsule, length=15.765, radius=9.3767, shapeName=spine_05_capsule, center=<5.8599 -3.5121 0>}
 ]
 thigh_l [{shapeType=Capsule, length=18.3081, radius=7.151, shapeName=thigh_l_capsule, center=<-14.3487 0 0>, rotation=<rot 90 0 0>}]
 thigh_r [{shapeType=Capsule, length=18.3081, radius=7.151, shapeName=thigh_l_capsule, center=<14.3487 0 0>, rotation=<rot 90 0 180>}]
 upperarm_l [{shapeType=Capsule, length=16.9273, radius=6.9167, shapeName=upperarm_l_capsule, center=<9.8981 0 0>, rotation=<rot -90 180 180>}]
 upperarm_r [{shapeType=Capsule, length=16.9273, radius=6.9167, shapeName=upperarm_l_capsule, center=<-9.8981 0 0>, rotation=<rot -90 180 0>}]
}
constraints [
 {
  bone1Name=head
  bone2Name=neck_02
  swing1LimitDegrees=45
  swing1Motion=Free
  swing2LimitDegrees=45
  swing2Motion=Free
  twistLimitDegrees=45
  twistMotion=Free
 }
 {bone1Name=Dyn_braid_01, bone2Name=Dyn_braid_Base, swing1LimitDegrees=22, swing2LimitDegrees=35, twistLimitDegrees=5}
 {
  bone1Name=spine_05
  bone2Name=spine_04
  swing1LimitDegrees=45
  swing1Motion=Free
  swing2LimitDegrees=45
  swing2Motion=Free
  twistLimitDegrees=45
  twistMotion=Free
 }
 {
  bone1Name=spine_03
  bone2Name=spine_02
  swing1LimitDegrees=45
  swing1Motion=Free
  swing2LimitDegrees=45
  swing2Motion=Free
  twistLimitDegrees=45
  twistMotion=Free
 }
 {bone1Name=Dyn_braid_02, bone2Name=Dyn_braid_01, swing1LimitDegrees=17, swing2LimitDegrees=25, twistLimitDegrees=5}
 {bone1Name=Dyn_braid_03, bone2Name=Dyn_braid_02, swing1LimitDegrees=12, swing2LimitDegrees=20, twistLimitDegrees=5}
 {bone1Name=Dyn_braid_04, bone2Name=Dyn_braid_03, swing1LimitDegrees=10, swing2LimitDegrees=10, twistLimitDegrees=0}
 {
  bone1Name=upperarm_l
  bone2Name=clavicle_l
  swing1LimitDegrees=45
  swing1Motion=Free
  swing2LimitDegrees=45
  swing2Motion=Free
  twistLimitDegrees=45
  twistMotion=Free
 }
 {
  bone1Name=clavicle_l
  bone2Name=spine_05
  swing1LimitDegrees=45
  swing1Motion=Free
  swing2LimitDegrees=45
  swing2Motion=Free
  twistLimitDegrees=45
  twistMotion=Free
 }
 {
  bone1Name=spine_04
  bone2Name=spine_03
  swing1LimitDegrees=45
  swing1Motion=Free
  swing2LimitDegrees=45
  swing2Motion=Free
  twistLimitDegrees=45
  twistMotion=Free
 }
 {
  bone1Name=neck_02
  bone2Name=neck_01
  swing1LimitDegrees=45
  swing1Motion=Free
  swing2LimitDegrees=45
  swing2Motion=Free
  twistLimitDegrees=45
  twistMotion=Free
 }
 {
  bone1Name=neck_01
  bone2Name=spine_05
  swing1LimitDegrees=45
  swing1Motion=Free
  swing2LimitDegrees=45
  swing2Motion=Free
  twistLimitDegrees=45
  twistMotion=Free
 }
 {
  bone1Name=lowerarm_l
  bone2Name=upperarm_l
  swing1LimitDegrees=45
  swing1Motion=Free
  swing2LimitDegrees=45
  swing2Motion=Free
  twistLimitDegrees=45
  twistMotion=Free
 }
 {
  bone1Name=Correct_hood_Back
  bone2Name=neck_01
  swing1LimitDegrees=45
  swing1Motion=Free
  swing2LimitDegrees=45
  swing2Motion=Free
  twistLimitDegrees=45
  twistMotion=Free
 }
 {
  bone1Name=upperarm_r
  bone2Name=spine_05
  swing1LimitDegrees=45
  swing1Motion=Free
  swing2LimitDegrees=45
  swing2Motion=Free
  twistLimitDegrees=45
  twistMotion=Free
 }
 {
  bone1Name=lowerarm_r
  bone2Name=upperarm_r
  swing1LimitDegrees=45
  swing1Motion=Free
  swing2LimitDegrees=45
  swing2Motion=Free
  twistLimitDegrees=45
  twistMotion=Free
 }
 {
  bone1Name=Dyn_braid_Base
  bone2Name=spine_05
  swing1LimitDegrees=45
  swing1Motion=Locked
  swing2LimitDegrees=45
  swing2Motion=Locked
  twistLimitDegrees=45
  twistMotion=Locked
 }
 {
  bone1Name=Base_hair_01
  bone2Name=Master_Correct_Hair
  swing1LimitDegrees=45
  swing1Motion=Free
  swing2LimitDegrees=45
  swing2Motion=Free
  twistLimitDegrees=45
  twistMotion=Free
 }
 {bone1Name=Dyn_hair_01, bone2Name=Base_hair_01, swing1LimitDegrees=20, swing2LimitDegrees=20, twistLimitDegrees=5}
 {bone1Name=Dyn_hair_02, bone2Name=Dyn_hair_01, swing1LimitDegrees=15, swing2LimitDegrees=15, twistLimitDegrees=3}
 {bone1Name=Dyn_hair_03, bone2Name=Dyn_hair_02, swing1LimitDegrees=10, swing2LimitDegrees=10, twistLimitDegrees=0}
 {
  bone1Name=Master_Correct_Hair
  bone2Name=head
  swing1LimitDegrees=45
  swing1Motion=Free
  swing2LimitDegrees=45
  swing2Motion=Free
  twistLimitDegrees=45
  twistMotion=Free
 }
]
