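# Post-copy AnimBlueprint for the DarkVogue female face accessory, adding secondary motion to hair.
# It layers the customization Body mesh's limb branches onto the attached parent's pose, then runs Control Rig, RigidBody and three chained Anim Dynamics hair bones with spherical limits.
# Anim Dynamics is active only when bRigidBodyEnabled is false; Control Rig also receives that flag and LOD2+.
# Unconnected duplicate physics nodes remain, and computed gravity override is unused by the live chain.
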AnimBlueprint /BRCosmetics/Characters/Player/Female/Medium/Bodies/F_MED_DarkVogue/Meshes/Parts/F_MED_DarkVogue_FaceAcc_AnimBP
target_skeleton=&/F_MED_DarkVogue_FaceAcc_Skeleton
graphs [
 AnimGraph {nodes [
   n1 = Anim Dynamics - Bone: Dyn_hair_01!(ComponentPose, Alpha=1, GravityOverride:Vector="0,0,0", ExternalForce="0,0,0") -> (Pose) {
    angularDampingOverride=1
    angularSpringConstant=500
    bAngularSpring=true
    bGravityOverrideInSimSpace=true
    bOverrideAngularDamping=true
    bUseGravityOverride=true
    bUsePlanarLimit=false
    bUseSphericalLimits=true
    simulationSpace=BoneRelative
    alphaBoolBlend {blendOption=Linear}
    boundBone {boneName=Dyn_hair_01}
    componentAppliedLinearAccClamp=<10000 10000 10000>
    componentLinearAccScale=<0.2 0.2 0.2>
    componentLinearVelScale=<0.2 0.2 0.2>
    physicsBodyDefinitions [{
      collisionType=CustomSphere
      sphereCollisionRadius=3
      boundBone {boneName=Dyn_hair_01}
      boxExtents=<5 5 5>
      constraintSetup {angularTargetAxis=AxisZ, angularLimitsMax=<10 5 15>, angularLimitsMin=<-10 -15 -15>, angularTarget=<0 0 1>}
      localJointOffset=<1 0 0>
     }]
    relativeSpaceBone {boneName=head}
    sphericalLimits [
     {limitRadius=7.5, drivingBone {boneName=upperarm_l}, sphereLocalOffset=<6.8964 0.0063 0.0024>}
     {limitRadius=8, drivingBone {boneName=clavicle_l}, sphereLocalOffset=<3.9581 -0.0107 0>}
     {limitRadius=3, drivingBone {boneName=Dyn_braid_01}, sphereLocalOffset=<3.8546 -0.4675 -1.1347>}
    ]
   }
   n2 = Anim Dynamics - Bone: Dyn_hair_02!(ComponentPose=@n1, Alpha=1, GravityOverride:Vector="0,0,0", ExternalForce="0,0,0") -> (Pose) {
    angularDampingOverride=1
    angularSpringConstant=500
    bAngularSpring=true
    bGravityOverrideInSimSpace=true
    bOverrideAngularDamping=true
    bUseGravityOverride=true
    bUsePlanarLimit=false
    bUseSphericalLimits=true
    simulationSpace=BoneRelative
    alphaBoolBlend {blendOption=Linear}
    boundBone {boneName=Dyn_hair_02}
    componentAppliedLinearAccClamp=<10000 10000 10000>
    componentLinearAccScale=<0.2 0.2 0.2>
    componentLinearVelScale=<0.2 0.2 0.2>
    physicsBodyDefinitions [{
      collisionType=CustomSphere
      sphereCollisionRadius=3
      boundBone {boneName=Dyn_hair_02}
      boxExtents=<5 5 5>
      constraintSetup {angularTargetAxis=AxisZ, angularLimitsMax=<8 3 13>, angularLimitsMin=<-8 -15 -13>, angularTarget=<0 0 1>}
      localJointOffset=<1 0 0>
     }]
    relativeSpaceBone {boneName=head}
    sphericalLimits [
     {limitRadius=4.8, drivingBone {boneName=upperarm_l}, sphereLocalOffset=<3.3031 0.0255 0.0373>}
     {limitRadius=8.2, drivingBone {boneName=clavicle_l}, sphereLocalOffset=<0 0 0>}
     {limitRadius=3, drivingBone {boneName=Dyn_braid_Base}, sphereLocalOffset=<4.6095 -0.2175 -0.7635>}
    ]
   }
   Anim Dynamics - Bone: Dyn_hair_03!(ComponentPose=@n2, Alpha=1, GravityOverride:Vector="0,0,0", ExternalForce="0,0,0") -> (Pose) {
    angularDampingOverride=1
    angularSpringConstant=500
    bAngularSpring=true
    bGravityOverrideInSimSpace=true
    bOverrideAngularDamping=true
    bUseGravityOverride=true
    bUsePlanarLimit=false
    bUseSphericalLimits=true
    simulationSpace=BoneRelative
    alphaBoolBlend {blendOption=Linear}
    boundBone {boneName=Dyn_hair_03}
    componentAppliedLinearAccClamp=<10000 10000 10000>
    componentLinearAccScale=<0.2 0.2 0.2>
    componentLinearVelScale=<0.2 0.2 0.2>
    physicsBodyDefinitions [{
      collisionType=CustomSphere
      sphereCollisionRadius=2.5
      boundBone {boneName=Dyn_hair_03}
      boxExtents=<5 5 5>
      constraintSetup {angularTargetAxis=AxisZ, angularLimitsMax=<5 0 10>, angularLimitsMin=<-5 -15 -10>, angularTarget=<0 0 1>}
      localJointOffset=<1 0 0>
     }]
    relativeSpaceBone {boneName=head}
    sphericalLimits [
     {limitRadius=7.2734, drivingBone {boneName=clavicle_l}, sphereLocalOffset=<4.6183 0 0>}
     {limitRadius=5, drivingBone {boneName=upperarm_l}, sphereLocalOffset=<8.4782 0 0.1196>}
     {limitRadius=2.5, drivingBone {boneName=Dyn_braid_02}, sphereLocalOffset=<2 0 0>}
    ]
   }
   RigidBody!(ComponentPose, Alpha, OverrideWorldGravity:Vector="0,0,-980", ExternalForce) -> (Pose) {
    bOverrideWorldGravity=true
    overridePhysicsAsset=&/F_MED_DarkVogue_FaceAcc_Physics
    simulationSpace=BaseBoneSpace
    alphaBoolBlend {blendOption=Linear}
    baseBoneRef {boneName=neck_01}
    componentAppliedLinearAccClamp=<5000 5000 5000>
    componentLinearAccScale {x=0.3}
    componentLinearVelScale {x=0.3}
    overrideWorldGravity {z=-980}
    simSpaceSettings {
     maxAngularAcceleration=500
     maxAngularVelocity=3
     maxLinearAcceleration=400
     maxLinearVelocity=250
     velocityScaleZ=0.2
     worldAlpha=0.8
     externalLinearDragV=<0.1 0.1>
    }
   }
   n3 = Layered blend per bone(
    BasePose=Copy Pose From Mesh() {bUseAttachedParent=true}
    BlendPoses_0=Copy Pose From Mesh(SourceMeshComponent=Get Body() -> (Body:Skeletal Mesh Component Object Reference))
   ) {layerSetup [{branchFilters [{blendDepth=2, boneName=clavicle_r}, {blendDepth=2, boneName=clavicle_l}, {blendDepth=1, boneName=thigh_l}, {blendDepth=1, boneName=thigh_r}]}]}
   n4 = Control Rig!(Source=@n3, Alpha, NOT_RigidBodyEnabled:bool=Get NOT_RigidBodyEnabled() -> (NOT_RigidBodyEnabled:bool), LOD2:bool=Get Lod2() -> (Lod2:bool)) -> (Pose) {
    alphaBoolBlend {blendOption=Linear}
    controlRigAssetReference {blueprintRigClass=&/F_MED_DarkVogue_FaceAcc_CtrlRig.F_MED_DarkVogue_FaceAcc_CtrlRig_C}
   }
   n5 = RigidBody(ComponentPose=Local To Component(LocalPose=@n4), OverrideWorldGravity="0,0,-980") {
    bOverrideWorldGravity=true
    overridePhysicsAsset=&F_MED_DarkVogue_FaceAcc_Physics
    simulationSpace=BaseBoneSpace
    alphaBoolBlend {blendOption=Linear}
    baseBoneRef {boneName=neck_01}
    componentAppliedLinearAccClamp=<5000 5000 5000>
    componentLinearAccScale=<0.4 0.1 0.1>
    componentLinearVelScale=<0.4 0.1 0.1>
    overrideWorldGravity {z=-980}
    simSpaceSettings {
     maxAngularAcceleration=500
     maxAngularVelocity=3
     maxLinearAcceleration=400
     maxLinearVelocity=250
     velocityScaleZ=0.2
     worldAlpha=0.8
     externalLinearDragV=<0.1 0.1>
    }
   }
   n6 = Get NOT_RigidBodyEnabled()
   n7 = Anim Dynamics - Bone: Dyn_hair_01!(ComponentPose=@n5, bAlphaBoolEnabled:bool=@n6, GravityOverride="0,0,0", ExternalForce="0,0,0") -> (Pose) {
    alphaInputType=Bool
    angularDampingOverride=1
    angularSpringConstant=500
    bAngularSpring=true
    bGravityOverrideInSimSpace=true
    bOverrideAngularDamping=true
    bUseGravityOverride=true
    bUsePlanarLimit=false
    bUseSphericalLimits=true
    simulationSpace=World
    alphaBoolBlend {blendOption=Linear}
    boundBone {boneName=Dyn_hair_01}
    componentAppliedLinearAccClamp=<10000 10000 10000>
    componentLinearAccScale=<0.2 0.2 0.2>
    componentLinearVelScale=<0.2 0.2 0.2>
    physicsBodyDefinitions [{
      collisionType=CustomSphere
      sphereCollisionRadius=3
      boundBone {boneName=Dyn_hair_01}
      boxExtents=<5 5 5>
      constraintSetup {angularTargetAxis=AxisZ, angularLimitsMax=<10 5 15>, angularLimitsMin=<-10 -12 -15>, angularTarget=<0 0 1>}
      localJointOffset=<0.4 0 0>
     }]
    relativeSpaceBone {boneName=head}
    sphericalLimits [
     {limitRadius=7.5, drivingBone {boneName=upperarm_l}, sphereLocalOffset=<6.8964 0.0063 0.0024>}
     {limitRadius=8, drivingBone {boneName=clavicle_l}, sphereLocalOffset=<3.9581 -0.0107 0>}
     {limitRadius=2.5, drivingBone {boneName=Dyn_braid_01}, sphereLocalOffset=<3.8546 -0.4675 -1.1347>}
     {limitRadius=10.6864, drivingBone {boneName=spine_05}, sphereLocalOffset=<4.0565 -6.5396 -8.175>}
    ]
   }
   n8 = Anim Dynamics - Bone: Dyn_hair_02!(ComponentPose=@n7, bAlphaBoolEnabled:bool=@n6, GravityOverride="0,0,0", ExternalForce="0,0,0") -> (Pose) {
    alphaInputType=Bool
    angularDampingOverride=1
    angularSpringConstant=500
    bAngularSpring=true
    bGravityOverrideInSimSpace=true
    bOverrideAngularDamping=true
    bUseGravityOverride=true
    bUsePlanarLimit=false
    bUseSphericalLimits=true
    simulationSpace=BoneRelative
    alphaBoolBlend {blendOption=Linear}
    boundBone {boneName=Dyn_hair_02}
    componentAppliedLinearAccClamp=<10000 10000 10000>
    componentLinearAccScale=<0.2 0.2 0.2>
    componentLinearVelScale=<0.2 0.2 0.2>
    physicsBodyDefinitions [{
      collisionType=CustomSphere
      sphereCollisionRadius=3
      boundBone {boneName=Dyn_hair_02}
      boxExtents=<5 5 5>
      constraintSetup {angularTargetAxis=AxisZ, angularLimitsMax=<8 3 13>, angularLimitsMin=<-8 -10 -13>, angularTarget=<0 0 1>}
      localJointOffset=<0.4 0 0>
     }]
    relativeSpaceBone {boneName=Dyn_hair_01}
    sphericalLimits [
     {limitRadius=4.8, drivingBone {boneName=upperarm_l}, sphereLocalOffset=<3.3031 0.0255 0.0373>}
     {limitRadius=8.2, drivingBone {boneName=clavicle_l}, sphereLocalOffset=<-0.0005 -1.0248 0>}
     {limitRadius=2, drivingBone {boneName=Dyn_braid_Base}, sphereLocalOffset=<4.6095 -0.2175 -0.7635>}
    ]
   }
   n9 = Anim Dynamics - Bone: Dyn_hair_03!(ComponentPose=@n8, bAlphaBoolEnabled:bool=@n6, GravityOverride="0,0,0", ExternalForce="0,0,0") -> (Pose) {
    alphaInputType=Bool
    angularDampingOverride=1
    angularSpringConstant=500
    bAngularSpring=true
    bGravityOverrideInSimSpace=true
    bOverrideAngularDamping=true
    bUseGravityOverride=true
    bUsePlanarLimit=false
    bUseSphericalLimits=true
    simulationSpace=BoneRelative
    alphaBoolBlend {blendOption=Linear}
    boundBone {boneName=Dyn_hair_03}
    componentAppliedLinearAccClamp=<10000 10000 10000>
    componentLinearAccScale=<0.2 0.2 0.2>
    componentLinearVelScale=<0.2 0.2 0.2>
    physicsBodyDefinitions [{
      collisionType=CustomSphere
      sphereCollisionRadius=2
      boundBone {boneName=Dyn_hair_03}
      boxExtents=<5 5 5>
      constraintSetup {angularTargetAxis=AxisZ, angularLimitsMax=<5 0 8>, angularLimitsMin=<-5 -8 -8>, angularTarget=<0 0 1>}
      localJointOffset=<0.4 0 0>
     }]
    relativeSpaceBone {boneName=Dyn_hair_02}
    sphericalLimits [
     {limitRadius=7.2734, drivingBone {boneName=clavicle_l}, sphereLocalOffset=<4.6183 0 0>}
     {limitRadius=5, drivingBone {boneName=upperarm_l}, sphereLocalOffset=<8.4782 0 0.1196>}
     {limitRadius=2, drivingBone {boneName=Dyn_braid_02}, sphereLocalOffset=<2 0 0>}
     {limitRadius=2, drivingBone {boneName=Dyn_braid_01}, sphereLocalOffset=<4 0 0>}
    ]
   }
   Output Pose(Result=Component To Local(ComponentPose=@n9)) {layerGroup=None}
  ]}
 EventGraph {nodes [
   Event BlueprintUpdateAnimation()
   > n1 = Gravity Override State Function DEPRECATED(
    CharacterPartAnimInstance=Self-Reference()
    GameTimeInSeconds=GetGameTimeInSeconds()
    GravityOverrideParams=Get Gravity Override Params()
    DebugJointName=dyn_hackBottomArmor
   )
   > Set Override World Gravity(Override World Gravity:Vector=@n1) -> (Output_Get:Vector)
   > Set NOT_RigidBodyEnabled(NOT_RigidBodyEnabled:bool=NOT Boolean(A=Get bRigidBodyEnabled())) -> (Output_Get:bool)
   > Set Lod2(Lod2:bool="integer >= integer"(A=Get CurrentLOD(), B=2)) -> (Output_Get:bool)
   n2 = Event OnCharacterCustomizationLoaded()
   > Set Body(Body:Skeletal Mesh Component Object Reference=Get Array Item with EFort Custom Part Type(Array=@n2.CustomMeshParts)) -> (Output_Get:Skeletal Mesh Component Object Reference)
  ]}
]
variables [
 Body {type=SkeletalMeshComponent}
 Gravity Override Params {
  type=GravityOverrideParamsStruct
  default {dBNO {jointName=root, gravityOverride=<0 0 -980>}, skydive_GoingUp {gravityOverride=<0 -300 800>}}
 }
 Lod2 {type=boolean}
 NOT_RigidBodyEnabled {type=boolean}
 Override World Gravity {type=Vector}
]
